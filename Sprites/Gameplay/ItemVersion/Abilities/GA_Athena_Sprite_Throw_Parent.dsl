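; Parent throw ability for Sprite items, launching an environmental Sprite from the player's inventory as a projectile.
; On commit, it finds the Sprite item, saves a drop entry, registers cooldown and tagged data via SpritesUtilityLibrary, then removes the item.
; The spawned projectile receives that saved entry and is flagged to execute the Sprite ability.
; Input press plays a failed sound for locally viewed pawns while cooldown tags are active.
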(event Ability|EventOnAbilityInputPressed
  (|CallParentFunction)
  (CallFunction|ShouldPlayFailedSound))

(event Ability|EventCommitExecute
  (bind _returnvalue (Ability|GetActorInfo))
  (bind (_owneractor _avataractor) (Utilities|Struct|BreakGameplayAbilityActorInfo _returnvalue))
  (bind _returnvalue_2 (Fortnite|GetFortPlayerControllerfromActor _avataractor))
  (bind _itemref (Class|BPLEnvironmentalSpriteFunctions|FindSpriteIteminInventory _returnvalue_2))
  (|CallParentFunction)
  (Utilities|IsValid _itemref
    (:Is%20Valid
      (bind _output_get (Variables|Default|SetItemEntryToPropagate (Item|CreateItemEntryforDrop _itemref)))
      (Variables|Default|SetbHasItemEntryToPropagate true)
      (Class|SpritesUtilityLibrary|RegisterCooldownandTaggedDatatoMap _returnvalue_2 _output_get)
      (bind _asfort_world_item_definition (Utilities|Casting|CastToFortWorldItemDefinition (Item|GetItemDefinition _itemref)))
      (Inventory|RemoveItemFromPlayer :PlayerController _returnvalue_2 :ItemDefinition _asfort_world_item_definition :AmountToRemove -1 :bForceRemoval true :bSkipVerbRemoval false))))

(event ProjectileLifecycle|EventOnProjectileSpawned
  (params (NewSpawnedProjectile :type Fort%20Projectile%20Base%20Object%20Reference))
  (|CallParentFunction NewSpawnedProjectile)
  (CallFunction|PropagateItemEntrytoProjectile NewSpawnedProjectile)
  (bind _asprj_athena_sprite_parent (Utilities|Casting|CastToPrj_Athena_Sprite_Parent NewSpawnedProjectile))
  (Variables|/Sprites/Gameplay/ItemVersion/Abilities/Prj_Athena_Sprite_Parent.Prj_Athena_Sprite_Parent_C|SetbExecuteSpriteAbility true _asprj_athena_sprite_parent))

(fn PropagateItemEntryToProjectile
  (params (InputPin :type Fort%20Projectile%20Base%20Object%20Reference))
  (if (Variables|Default|GetbHasItemEntryToPropagate)
    (Class|SpritesUtilityLibrary|SpritesTryAddItemEntrytoActor (Variables|Default|GetItemEntryToPropagate) InputPin)))

(fn ShouldPlayFailedSound
  (Networking|CanExecuteCosmeticEvents
    (:True
      (bind _outfortplayerpawn (CallFunction|GetAvatarasValidFortPlayerPawn)
        (:ValidFortPlayerPawn
          (if (Pawn|IsLocallyViewed _outfortplayerpawn)
            (bind _returnvalue (Ability|Cooldowns|IsCooldownTagsActive _outfortplayerpawn (Ability|Cooldowns|GetCooldownTags)))
            (if _returnvalue
              (Utilities|IsValid _outfortplayerpawn
                (:Is%20Valid
                  (Audio|PlaySoundatLocation (Variables|Default|GetFailedSound) (Transformation|GetActorLocation _outfortplayerpawn) "0, 0, 0" 1.0 1.0 0.0 0 0 _outfortplayerpawn))))))))))
