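; Gameplay ability that shows a Sprite as a player backpack with animated reactions.
; On activation, BP_SpriteBackpackController spawns it under a GUID, replicated via the server and locally on the owning client; ending clears it.
; Equipped-tag changes toggle its sound library and bind reactions only while unequipped.
; Montages react to crouches, eliminations, random jumps and health or shield drops against timer-polled values, while DBNO crying listens throughout.
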(event Ability|EventActivateAbility
  ;[1 1 1 0.5] Activate Ability
  (|CallParentFunction)
  (Variables|Default|SetAbilityGUID (Guid|NewGuid))
  (bind _asfort_player_pawn (Utilities|Casting|CastToFortPlayerPawn (Ability|GetAvatarActorfromActorInfo)))
  (bind _output_get (Variables|Default|SetBackpackController (Actor|GetComponentbyClass _asfort_player_pawn "/Sprites/Gameplay/Components/BP_SpriteBackpackController.BP_SpriteBackpackController_C")))
  (Default|AssignHiddenChanged _output_get (delegate Custom|HiddenChanged_Event))
  (Ability|HasAuthority
    (:True
      (Utilities|IsValid (Variables|Default|GetBackpackController)
        (:Is%20Valid
          (Class|BPSpriteBackpackController|SpawnBackpackServer (Variables|Default|GetBackpackController) (Variables|Default|GetBackpackClass) (Variables|Default|GetAbilityGUID) :!pinned true) ; Sets class to be replicated to simulated proxies
        )))
    (:False
      (Ability|IsLocallyControlled
        (:True
          (bind _backpackactor (Default|LocallySpawnBackpack (Variables|Default|GetBackpackController) (Variables|Default|GetBackpackClass) (Variables|Default|GetAbilityGUID)))
          (Variables|Default|SetBackpackActorRef _backpackactor)))))
  (CallFunction|AddTagCountChangedEventListener (Variables|GameplayAbility_SpriteBackpack|GetEquippedTag) (delegate Custom|HandleEquipTag))
  (CallFunction|AddTagAddedEventListener (Variables|Default|GetDBNOEventTag) (delegate Custom|HandleDBNO)))

(event Custom|HandleDBNO
  ;[1 1 1 0.5] Activate Ability
  (CallFunction|PlayCryMontage))

(event Custom|HiddenChanged_Event
  (params (NewHidden :type Boolean))
  ;[1 1 1 0.5] Activate Ability
  (bind _backpackactorref (Variables|Default|GetBackpackActorRef)
    (:then
      (Class|BSpriteHeldBackpack|OnHiddenChanged _backpackactorref NewHidden))
    (:else)))

(event Custom|HandleEquipTag
  (params (TagCount :type Integer))
  ;[1 1 1 0.5] Activate Ability
  (Utilities|Casting|CastToFortPlayerPawn (Ability|GetAvatarActorfromActorInfo))
  (CallFunction|UpdateMusicTags (> TagCount 0))
  (if (<= TagCount 0)
    (Utilities|FlowControl|Delay 0.5)
    (CallFunction|BindAnimationEvents)
    (else
      (CallFunction|UnbindandClearTasks))))

(event Custom|UnbindAndClearTasks
  ;[1 1 1 0.5] End Ability
  (Utilities|IsValid (stale Ability|EventOnEndAbility _asfort_player_pawn)
    (:Is%20Valid
      (Health|UnbindEventfromOnPawnHealthChanged (stale Ability|EventOnEndAbility _asfort_player_pawn) (EventDispatchers|CreateEvent :!event Custom|PlayAgitatedMontage)))
    (:Is%20Not%20Valid
      _))
  (CallFunction|RemoveListenerbyTag (Variables|Default|GetEliminationEventTag))
  (CallFunction|RemoveListenerbyTag (Variables|Default|GetCrouchEventTag))
  (CallFunction|RemoveListenerbyTag (Variables|Default|GetJumpEventTag))
  (Utilities|Time|ClearandInvalidateTimerbyHandle (Variables|Default|GetHealthTimer)))

(event Ability|EventOnEndAbility
  (params (bWasCancelled :type Boolean))
  ;[1 1 1 0.5] End Ability
  (|CallParentFunction bWasCancelled)
  (Utilities|IsValid (Variables|Default|GetBackpackController)
    (:Is%20Valid
      (Ability|HasAuthority
        (:True
          (Class|BPSpriteBackpackController|ClearBackpackServer (Variables|Default|GetBackpackController) (Variables|Default|GetAbilityGUID)))
        (:False
          (Class|BPSpriteBackpackController|DestroyBackpack (Variables|Default|GetBackpackController) (Variables|Default|GetAbilityGUID)))))
    (:Is%20Not%20Valid
      _))
  (bind _asfort_player_pawn (Utilities|Casting|CastToFortPlayerPawn (Ability|GetAvatarActorfromActorInfo)))
  (CallFunction|UpdateMusicTags true)
  (Health|UnbindEventfromOnPawnHealthChanged _asfort_player_pawn (EventDispatchers|CreateEvent :!event Custom|PlayAgitatedMontage))
  (CallFunction|RemoveListenerbyTag (Variables|Default|GetEliminationEventTag))
  (CallFunction|RemoveListenerbyTag (Variables|Default|GetCrouchEventTag))
  (CallFunction|RemoveListenerbyTag (Variables|Default|GetJumpEventTag))
  (Utilities|Time|ClearandInvalidateTimerbyHandle (Variables|Default|GetHealthTimer)))

(event Custom|BindAnimationEvents
  ;[1 1 1 0.5] Bind to Animation Events
  (bind _asfort_player_pawn (Utilities|Casting|CastToFortPlayerPawn (Ability|GetAvatarActorfromActorInfo)))
  (bind _returnvalue (Utilities|Time|SetTimerbyEvent (delegate Custom|GetHealthValues) 2.0 true))
  (Variables|Default|SetHealthTimer _returnvalue)
  (Health|BindEventtoOnPawnHealthChanged _asfort_player_pawn (delegate Custom|PlayAgitatedMontage))
  (CallFunction|AddGameplayEventListener (Variables|Default|GetEliminationEventTag) (delegate Custom|HandleElimination))
  (CallFunction|AddTagAddedEventListener (Variables|Default|GetCrouchEventTag) (delegate Custom|HandleCrouch))
  (CallFunction|AddGameplayEventListener (Variables|Default|GetJumpEventTag) (delegate Custom|HandleJump)))

(event Custom|HandleJump
  (params (Tag :type Gameplay%20Tag%20Structure) (Data :type Gameplay%20Event%20Data%20Structure%20%28by%20ref%29))
  ;[1 1 1 0.5] Bind to Animation Events
  (if (Math|Float|InRange(Float) (Math|Random|RandomFloatinRange 1.0 1e+02) 8e+01 1e+02)
    (CallFunction|PlayCelebrateMontage)))

(event Custom|HandleCrouch
  ;[1 1 1 0.5] Bind to Animation Events
  (CallFunction|PlayCrouchMontage))

(event Custom|HandleElimination
  (params (Tag :type Gameplay%20Tag%20Structure) (Data :type Gameplay%20Event%20Data%20Structure%20%28by%20ref%29))
  ;[1 1 1 0.5] Bind to Animation Events
  (CallFunction|PlayCelebrateMontage))

(event Custom|PlayAgitatedMontage
  ;[1 1 1 0.5] Bind to Animation Events
  (if (Health|IsDBNO (stale Custom|BindAnimationEvents _asfort_player_pawn))
    _
    (elif (or (< (FortPawn|GetHealth (stale Custom|BindAnimationEvents _asfort_player_pawn)) (stale Custom|GetHealthValues _output_get)) (< (FortPawn|GetShield (stale Custom|BindAnimationEvents _asfort_player_pawn)) (stale Custom|GetHealthValues _output_get_2)))
      (Utilities|FlowControl|Delay 0.5)
      (Class|BSpriteHeldBackpack|PlayAgitatedMontage (Variables|Default|GetBackpackActorRef)))))

(event Custom|GetHealthValues
  ;[1 1 1 0.5] Bind to Animation Events
  (bind _output_get (Variables|Default|SetCurrentHealth (FortPawn|GetHealth (stale Custom|BindAnimationEvents _asfort_player_pawn))))
  (bind _output_get_2 (Variables|Default|SetCurrentShield (FortPawn|GetShield (stale Custom|BindAnimationEvents _asfort_player_pawn)))))

(event Custom|PlayCryMontage
  ;[0.05838 1 0.1411 0.5] Play the Montages
  (Utilities|FlowControl|Delay 0.5)
  (Class|BSpriteHeldBackpack|PlayCryMontage (Variables|Default|GetBackpackActorRef)))

(event Custom|UpdateMusicTags
  (params (bBackpackHidden :type Boolean))
  ; Adds/removes required soundlibraries for sprites when equipped as backpacks
  (bind _backpackactorref (Variables|Default|GetBackpackActorRef)
    (:then
      (if bBackpackHidden
        (SoundLibrary|RemoveLibrary (Variables|/Sprites/Gameplay/ItemVersion/B_Sprite_Held_Backpack.B_Sprite_Held_Backpack_C|GetSoundLibrary _backpackactorref) (Variables|/Sprites/Gameplay/ItemVersion/B_Sprite_Held_Backpack.B_Sprite_Held_Backpack_C|GetSpriteSoundLibrary _backpackactorref) :!class /Script/SoundLibrary.SoundLibraryComponent)
        (else
          (SoundLibrary|AddLibrary (Variables|/Sprites/Gameplay/ItemVersion/B_Sprite_Held_Backpack.B_Sprite_Held_Backpack_C|GetSoundLibrary _backpackactorref) (Variables|/Sprites/Gameplay/ItemVersion/B_Sprite_Held_Backpack.B_Sprite_Held_Backpack_C|GetSpriteSoundLibrary _backpackactorref) :!class /Script/SoundLibrary.SoundLibraryComponent))))
    (:else)))

(event Custom|PlayCrouchMontage
  ;[0.05838 1 0.1411 0.5] Play the Montages
  (Utilities|FlowControl|Delay 0.5)
  (Class|BSpriteHeldBackpack|PlayCrouchMontage (Variables|Default|GetBackpackActorRef)))

(event Custom|PlayCelebrateMontage
  ;[0.05838 1 0.1411 0.5] Play the Montages
  (Utilities|FlowControl|Delay 0.5)
  (Class|BSpriteHeldBackpack|PlayCelebrateMontage (Variables|Default|GetBackpackActorRef)))

(fn OnRep_IsEquipped)
