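; Parent projectile for thrown Sprite items in Fortnite, which bounces and then spawns an awake or sleeping sprite actor.
; Hit validation combines bounce limits with slope, surface type, awake-physics, obstacle and blocking-volume checks.
; On authority a valid hit initiates the spawn; remotes locally disable the projectile.
; Spawning can be deferred via a rollback CVar, propagates the item entry and hands tracking to the sprite, flagged thrown or dropped.
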(fn CanSpawnSpriteAt
  (params (Hit :type Hit%20Result%20Structure))
  (bind _row_obstacleboxtrace_zoffset (Variables|Default|GetRow_ObstacleBoxTrace_ZOffset))
  (bind _returnvalue (ScalableFloat|GetValueatLevel _row_obstacleboxtrace_zoffset))
  (bind _row_obstacleboxtrace_halfsize (Variables|Default|GetRow_ObstacleBoxTrace_HalfSize))
  (bind _returnvalue_2 (ScalableFloat|GetValueatLevel _row_obstacleboxtrace_halfsize))
  (bind _returnvalue_3 (+ _returnvalue_2 _returnvalue))
  (bind _hit (Variables|GetHit))
  (bind (_bblockinghit _binitialoverlap _time _distance _location _impactpoint _normal _impactnormal) (Collision|BreakHitResult _hit))
  (bind _returnvalue_4 (* _impactnormal _returnvalue_3))
  (bind (_bblockinghit_2 _binitialoverlap_2 _time_2 _distance_2 _location_2 _impactpoint_2 _normal_2 _impactnormal_2 _physmat _hitactor) (Collision|BreakHitResult Hit))
  (bind _returnvalue_5 (Math|Vector|VectorUp))
  (bind _returnvalue_6 (Math|Vector|DotProduct _impactnormal_2 _returnvalue_5))
  (bind _hit_2 (Variables|GetHit))
  (bind (_bblockinghit_3 _binitialoverlap_3 _time_3 _distance_3 _location_3 _impactpoint_3 _normal_3 _impactnormal_3 _physmat_2 _hitactor_2) (Collision|BreakHitResult _hit_2))
  (bind _returnvalue_7 (Actor|GetComponentbyClass _hitactor_2 "/Script/FortniteGame.FortPhysicsObjectComponent"))
  (bind (_asbuilding_actor _bsuccess) (Utilities|Casting|CastToBuildingActor _hitactor))
  (bind (_asspline_mesh_actor _bsuccess_2) (Utilities|Casting|CastToSplineMeshActor _hitactor))
  (bind (_asbuilding_wall _bsuccess_3) (Utilities|Casting|CastToBuildingWall _hitactor))
  (bind (_asbuilding_gameplay_actor _bsuccess_4) (Utilities|Casting|CastToBuildingGameplayActor _hitactor))
  (bind (_aslandscape_spline_actor _bsuccess_5) (Utilities|Casting|CastToLandscapeSplineActor _hitactor))
  (bind (_ascontrol_point_mesh_actor _bsuccess_6) (Utilities|Casting|CastToControlPointMeshActor _hitactor))
  (bind (_aslandscape_spline_meshes_actor _bsuccess_7) (Utilities|Casting|CastToLandscapeSplineMeshesActor _hitactor))
  (bind (_aslandscape _bsuccess_8) (Utilities|Casting|CastToLandscape _hitactor))
  (bind (_aslandscape_streaming_proxy _bsuccess_9) (Utilities|Casting|CastToLandscapeStreamingProxy _hitactor))
  (bind (_asstatic_mesh_actor _bsuccess_10) (Utilities|Casting|CastToStaticMeshActor _hitactor))
  (bind (_asbuilding_deco _bsuccess_11) (Utilities|Casting|CastToBuildingDeco _hitactor))
  (bind _returnvalue_8 (+ _impactpoint _returnvalue_4))
  ;[0.2446 0.3583 1] Slope Check
  (if (and (>= _returnvalue_6 (Math|Trig|Cos(Degrees) (ScalableFloat|GetValueatLevel (Variables|Default|GetRow_MaxSlopeAngle)))) (> (Math|Trig|Cos(Degrees) (ScalableFloat|GetValueatLevel (Variables|Default|GetRow_MaxSlopeAngle))) 0))
    ;[0.02517 0.4315 1] Valid surface actor check
    (if (and (Math|Boolean|ORBoolean _bsuccess_9 _bsuccess_8 _bsuccess_10 _bsuccess_2 _bsuccess_6 _bsuccess _bsuccess_5 _bsuccess_7) (not (Math|Boolean|ORBoolean _bsuccess_11 _bsuccess_4 _bsuccess_3)))
      ;[0.524 0.9834 1] Awake physics check
      ; Placement on actors with FortPhycisObjectComponents is acceptable as long as the physics component is not awake. (The physics actor waking while the turret is on it is handled)
      (Utilities|IsValid _returnvalue_7
        (:Is%20Valid
          (if (PhysicsObject|IsAwake _returnvalue_7)
            (Development|PrintString (Utilities|String|Append "Turret Placement Failed - Physic Awake -" (Utilities|GetDisplayName _hitactor_2)) (Variables|Default|GetbDebug) (Variables|Default|GetbDebug))
            (else
              ;[0.1012 0.1344 1] ObstacleOverlapcheck
              (bind (_outhit _returnvalue_9) (Collision|BoxTraceByChannel :Start _returnvalue_8 :End _returnvalue_8 :HalfSize (Math|Vector|MakeVector _returnvalue_2 _returnvalue_2 _returnvalue_2) :Orientation (Math|Rotator|MakeRotfromZ _impactnormal) :TraceChannel "TraceTypeQuery3" :bTraceComplex false :DrawDebugType (select (Variables|Default|GetbDebug) "ForDuration" "None") :bIgnoreSelf true :TraceColor "(R=1.000000,G=0.000000,B=0.000000,A=1.000000)" :TraceHitColor "(R=0.000000,G=1.000000,B=0.000000,A=1.000000)" :DrawTime 5.0 :!label _label)) ; Use FCC_Trace_Interaction to detect obstacles\nTrace in a box slightly off of and aligned with the hit surface
              (if (and _returnvalue_9 (Collision|BreakHitResult _outhit))
                (Development|PrintString (Utilities|String|Append "Turret Placement Failed - Obstacle Detected - " (Utilities|GetDisplayName (.HitActor (Collision|BreakHitResult _outhit)))) (Variables|Default|GetbDebug) (Variables|Default|GetbDebug))
                (else
                  ;[0.4195 0.5318 1] Overlap Blocking Volume Check
                  (bind _boverlapping (Default|IsOverlappingDeployableBlockVolume self (ScalableFloat|GetValueatLevel (Variables|Default|GetRow_ObstacleBoxTrace_HalfSize)) _returnvalue_8))
                  (if _boverlapping
                    (Development|PrintString "Turret Placement Failed - Blocking Volume Detected" (Variables|Default|GetbDebug) (Variables|Default|GetbDebug))
                    (else
                      ;[0.3174 1 0.4881] Return VALID
                      (return true)))))))
          ;[1 0.5697 0.2864] Return INVALID
          (return false))
        (:Is%20Not%20Valid
          _))
      (goto _label)
      (else
        (Development|PrintString (Utilities|String|Append "Turret Placement Failed - Invalid Hit Actor -" (Utilities|GetDisplayName _hitactor)) (Variables|Default|GetbDebug) (Variables|Default|GetbDebug))))
    (else
      (Development|PrintString (Utilities|String|Append "Turret Placement Failed - Invalid Slope - Dot " (Utilities|String|ToString(Float) _returnvalue_6)) (Variables|Default|GetbDebug) (Variables|Default|GetbDebug))))
  (return false))

(fn CheckAngleForBounce
  (params (A :type Vector) (A2 :type Vector))
  (return (/ (Math|Float|Lerp (ScalableFloat|GetValueatLevel (Variables|Default|GetHF_MaxBounceAtShallowAngle)) (ScalableFloat|GetValueatLevel (Variables|Default|GetHF_MaxBounceAtSteepAngle)) (Math|Float|NormalizetoRange (Math|Float|%25%28Float%29 (Math|Trig|Acos(Degrees) (Math|Vector|DotProduct A2 (Math|Vector|Normalize (Variables|/Script/Engine.MovementComponent|GetVelocity (Variables|Movement|GetProjectileMovementComponent))))) 91.0) 0.0 9e+01)) (Math|Vector|VectorLength (Variables|/Script/Engine.MovementComponent|GetVelocity (Variables|Movement|GetProjectileMovementComponent)))) (>= (Math|Vector|VectorLength (* A (* (Math|Vector|VectorLength (Variables|/Script/Engine.MovementComponent|GetVelocity (Variables|Movement|GetProjectileMovementComponent))) (Variables|/Script/Engine.ProjectileMovementComponent|GetBounciness (Variables|Movement|GetProjectileMovementComponent))))) (Math|Float|Lerp (ScalableFloat|GetValueatLevel (Variables|Default|GetHF_MaxBounceAtShallowAngle)) (ScalableFloat|GetValueatLevel (Variables|Default|GetHF_MaxBounceAtSteepAngle)) (Math|Float|NormalizetoRange (Math|Float|%25%28Float%29 (Math|Trig|Acos(Degrees) (Math|Vector|DotProduct A2 (Math|Vector|Normalize (Variables|/Script/Engine.MovementComponent|GetVelocity (Variables|Movement|GetProjectileMovementComponent))))) 91.0) 0.0 9e+01)))))

(fn DeferBGASpawn
  (params (BGAToSpawn :type Actor%20Class%20Reference) (Normal :type Vector) (HitActor :type Actor%20Object%20Reference) (HitComponent :type Primitive%20Component%20Object%20Reference) (Location :type Vector) (bIsFromWater :type Boolean) (Hit :type Hit%20Result%20Structure))
  (if (Class|FortItemEntryComponent|HasItem (Variables|Default|GetFortItemEntry))
    (return true)
    (else
      (CallFunction|LocallyDisableProjectile)
      (if (and (not (Utilities|IsValidClass (Variables|Default|GetDeferredSpawnBGAClass))) (Math|Float|NearlyEqual(Float) (Fortnite|GetNumericCVar "Fort.Rollback.DeferSpriteSpawnFromProjectile") 1.0 0.001))
        (Variables|Default|SetDeferredSpawnData (Utilities|Struct|MakeDeferredSpawnParameters Normal HitActor HitComponent Location bIsFromWater Hit))
        (Variables|Default|SetDeferredSpawnBGAClass BGAToSpawn)
        (return false)
        (else
          (return false))))))

(fn HandleHit
  (params (OtherActor :type Actor%20Object%20Reference) (OtherComp :type Primitive%20Component%20Object%20Reference) (Hit :type Hit%20Result%20Structure%20%28by%20ref%29) (HitLocation :type Vector :default "0, 0, 0"))
  (bind _validspawn (CallFunction|IsHitValidforSpawn OtherComp OtherActor Hit))
  (if _validspawn
    (Networking|SwitchHasAuthority
      (:Authority
        (CallFunction|InitiateSpawnBGA Hit OtherActor OtherComp HitLocation))
      (:Remote
        (CallFunction|LocallyDisableProjectile)))))

(fn IsHitValidForSpawn
  (params (HitOtherComp :type Primitive%20Component%20Object%20Reference) (HitOther :type Actor%20Object%20Reference) (HitResult :type Hit%20Result%20Structure))
  (bind (_bblockinghit _binitialoverlap _time _distance _location _impactpoint _normal _impactnormal _physmat _hitactor _hitcomponent) (Collision|BreakHitResult HitResult))
  (if (>= (Variables|Default|GetBounce%20Count) (ScalableFloat|GetValueasInteger (Variables|Default|GetRow_MinBounces)))
    (bind _isvalid (CallFunction|CanSpawnSpriteAt HitResult))
    (if _isvalid
      (return true false)
      (elif (or (> (Variables|Default|GetBounce%20Count) (ScalableFloat|GetValueasInteger (Variables|Default|GetRow_MaxBounces))) (Utilities|Enum|Equal(Enum) (Collision|GetCollisionResponsetoChannel _hitcomponent "ECC_GameTraceChannel15") "ECR_Block"))
        (return true false)
        (else
          (return false false))))
    (else
      (return false false))))

(fn OnSpriteItemDropped
  (params (ItemEntry :type Fort%20Item%20Entry%20Structure%20%28by%20ref%29))
  (bind _returnvalue (CallFunction|GetOwnerWeapon))
  (if (== (Data|GetWeaponData _returnvalue) (Class|FortItemFunctionLibrary|BreakFortItemEntry ItemEntry))
    (CallFunction|UnbindfromItemDroppedDelegate)
    (Variables|Default|SetbSpawningFromItemDrop true)
    (return)))

(fn SpawnBGA_
  (params (Normal :type Vector :default "0, 0, 0") (HitActor :type Actor%20Object%20Reference) (HitComp :type Primitive%20Component%20Object%20Reference) (Pos :type Vector :default "0, 0, 0") (FromWater :type Boolean) (Hit :type Hit%20Result%20Structure))
  (bind _returnvalue (Transformation|GetActorLocation))
  (bind (_returnvalue_roll _returnvalue_pitch _returnvalue_yaw) (Math|Rotator|BreakRotator (Math|Rotator|FindLookatRotation (Variables|Default|GetInstigatorLocation) (Variables|Default|GetThrowEndLocation))))
  (bind _instigator (Variables|Actor|GetInstigator))
  (bind (_returnvalue_roll_2 _returnvalue_pitch_2 _returnvalue_yaw_2) (Math|Rotator|BreakRotator (Pawn|GetBaseAimRotation _instigator)))
  (if (not (Variables|Default|GetSkipDeferBgaSpawn))
    (bind _returnvalue_2 (CallFunction|DeferBGASpawn (Variables|Default|GetSpriteActorToSpawn) Normal HitActor HitComp Pos FromWater Hit))
    (if _returnvalue_2
      _))
  (Utilities|FlowControl|DoOnce false
    (:Completed
      (Utilities|Time|InvalidateTimerHandle (Variables|Default|GetStuckCheckTimer))
      (if (not (Variables|Default|GetHasSpawnedAI?))
        (if (Networking|IsServer)
          (Utilities|IsValid _instigator
            (:Is%20Valid
              (CallFunction|SpawnSprite (Math|Transform|MakeTransform _returnvalue (Math|Rotator|MakeRotator 0.0 0.0 (+ _returnvalue_yaw_2 9e+01)) "1.000000,1.000000,1.000000")))
            (:Is%20Not%20Valid
              (CallFunction|SpawnSprite (Math|Transform|MakeTransform _returnvalue (Math|Rotator|MakeRotator 0.0 0.0 (+ _returnvalue_yaw -9e+01)) "1.000000,1.000000,1.000000"))))
          (bind _returnvalue_3 (CallFunction|GetOwnerWeapon))
          (Utilities|IsValid _returnvalue_3
            (:Is%20Valid
              (bind _spawnedsprite (Variables|Default|GetSpawnedSprite_)
                (:then
                  (Class|BPSpriteTrackingComponent|TransferTrackingOwnership (Actor|GetComponentbyClass _returnvalue_3 "/Sprites/Gameplay/Components/BP_SpriteTrackingComponent.BP_SpriteTrackingComponent_C") (Actor|GetComponentbyClass _spawnedsprite "/Sprites/Gameplay/Components/BP_SpriteTrackingComponent.BP_SpriteTrackingComponent_C"))
                  (if (Variables|Default|GetbSpawningFromItemDrop)
                    (Default|CallOnSpriteDropped (Actor|GetComponentbyClass _spawnedsprite "/Sprites/Gameplay/Components/BP_SpriteTrackingComponent.BP_SpriteTrackingComponent_C"))
                    (else
                      (Default|CallOnSpriteThrown (Actor|GetComponentbyClass _spawnedsprite "/Sprites/Gameplay/Components/BP_SpriteTrackingComponent.BP_SpriteTrackingComponent_C"))))
                  (return (Variables|Default|GetSpawnedSprite_)))
                (:else)))))))))

(fn SpawnBGA_Sleeping
  (params (Normal :type Vector :default "0, 0, 0") (HitActor :type Actor%20Object%20Reference) (HitComp :type Primitive%20Component%20Object%20Reference) (Pos :type Vector :default "0, 0, 0") (FromWater :type Boolean) (Hit :type Hit%20Result%20Structure))
  (bind (_returnvalue_roll _returnvalue_pitch _returnvalue_yaw) (Math|Rotator|BreakRotator (Pawn|GetBaseAimRotation (Variables|Actor|GetInstigator))))
  (bind _returnvalue (CallFunction|DeferBGASpawn (Variables|Default|GetSleepingSpriteToSpawn) Normal HitActor HitComp Pos FromWater Hit))
  (if _returnvalue
    (Utilities|FlowControl|DoOnce false
      (:Completed
        (Utilities|Time|InvalidateTimerHandle (Variables|Default|GetStuckCheckTimer))
        (if (not (Variables|Default|GetHasSpawnedAI?))
          (if (Networking|IsServer)
            (Default|SpawnSpriteSleeping (Math|Transform|MakeTransform (Transformation|GetActorLocation) (Math|Rotator|MakeRotator 0.0 0.0 (+ _returnvalue_yaw 9e+01)) "1.000000,1.000000,1.000000"))
            (return 0)))))))

(fn SpawnSprite
  (params (Transform :type Transform))
  ;[1 0.3016 0.2038] Spawn sprite, either from prj stop or if it hits a shrine
  (Variables|Default|SetHasSpawnedAI? true)
  (bind _returnvalue (Game|SpawnActorfromClass (Variables|Default|GetSpriteActorToSpawn) Transform "AdjustIfPossibleButAlwaysSpawn" "MultiplyWithRoot" 0 (Game|GetInstigator)))
  (bind _asbga_sprite_parent (Utilities|Casting|CastToBGA_Sprite_Parent _returnvalue)
    (:then
      (bind _output_get (Variables|Default|SetSpawnedSprite_ _asbga_sprite_parent))
      (Class|SpritesUtilityLibrary|SpritesTryPropagateItemEntrytoNextActor self _output_get)
      (Variables|/Sprites/Gameplay/BGA_Sprite_Parent.BGA_Sprite_Parent_C|SetbExecuteSpriteAbility (Variables|Default|GetbExecuteSpriteAbility) _asbga_sprite_parent))
    (:CastFailed
      _))
  (Rendering|SetActorHiddenInGame true)
  (Collision|SetActorEnableCollision false)
  (Actor|SetLifeSpan 1.0)
  (CallFunction|StopinAirLoop)
  (return))

(fn SpawnSprite_Sleeping
  (params (Transform :type Transform))
  ;[1 0.3016 0.2038] Spawn sprite, either from prj stop or if it hits a shrine
  (Variables|Default|SetHasSpawnedAI? true)
  (bind _returnvalue (Game|SpawnActorfromClass (Variables|Default|GetSleepingSpriteToSpawn) Transform "AdjustIfPossibleButAlwaysSpawn" "MultiplyWithRoot" 0 (Game|GetInstigator)))
  (bind _output_get (Variables|Spawning|SetSpawnedBGA _returnvalue))
  (Actor|SetLifeSpan _output_get (ScalableFloat|GetCachedScalableFloatValue (Variables|Default|GetHF_LifespanOnTrain)))
  (bind _asbga_sprite_parent (Utilities|Casting|CastToBGA_Sprite_Parent _returnvalue)
    (:then
      (bind _output_get_2 (LocalVariables|SetSpawnedSprite _asbga_sprite_parent))
      (Class|SpritesUtilityLibrary|SpritesTryPropagateItemEntrytoNextActor self _output_get_2))
    (:CastFailed
      _))
  (CallFunction|StopinAirLoop)
  (Rendering|SetActorHiddenInGame true)
  (Collision|SetActorEnableCollision false)
  (Actor|SetLifeSpan 1.0)
  (return))

(fn UnbindFromItemDroppedDelegate
  (bind _returnvalue (CallFunction|GetOwnerWeapon))
  (Utilities|IsValid _returnvalue
    (:Is%20Valid
      (bind _asfort_player_pawn (Utilities|Casting|CastToFortPlayerPawn (Game|GetInstigator _returnvalue)))
      (FortPawn|UnbindEventfromOnItemDroppedDelegate _asfort_player_pawn (EventDispatchers|CreateEvent :!event OnSpriteItemDropped)))))

(fn ConstructionScript
  (|CallParentFunction)
  (Collision|SetCollisionResponsetoChannel (Variables|Collision|GetCapsuleComponent) "ECC_GameTraceChannel18" "ECR_Overlap"))
